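# Editor utility generated-mesh actor that bakes DelMar track layout Blueprints into combined render and collision static meshes.
# It gathers each layout's StaticMeshComponents, remaps material IDs to a shared list, and appends transformed meshes per LOD with Geometry Script.
# Render LODs get planar-projected road UVs and Nanite, with screen sizes from optional per-layout overrides read via Python.
# Collision-enabled components form a separate complex-collision mesh assigned to the render mesh.
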EditorUtilityBlueprint /DelMarTools/EditorUtilities/GeometryScripts/TrackCombiner/EdGenMesh_TrackCombiner
graphs [
 ApplyMaterialIDsToMesh {nodes [
   n1 = ApplyMaterialIDsToMesh!() -> (
    then
    TargetMesh:Dynamic Mesh Object Reference
    MatArray:Array of Material Interface Object References
    Comp:Static Mesh Component Object Reference
   ) {is=title K2Node_FunctionEntry}
   > Set l_targetMesh(l_targetMesh:Dynamic Mesh Object Reference=@n1) -> (Output_Get:Dynamic Mesh Object Reference)
   > Set l_matArray(l_matArray:Array of Material Interface Object References=@n1.MatArray) -> (Output_Get:Array of Material Interface Object References)
   > Set l_comp(l_comp:Static Mesh Component Object Reference=@n1.Comp) -> (Output_Get:Static Mesh Component Object Reference)
   > n2 = Sequence()
   > n3 = AllocateComputeMesh!(execute, self:Dynamic Mesh Actor Object Reference) -> (then, ReturnValue:Dynamic Mesh Object Reference)
   > Set l_outputMesh(l_outputMesh:Dynamic Mesh Object Reference=@n3) -> (Output_Get:Dynamic Mesh Object Reference)
   > n4 = GetComputeMeshPool!(execute, self:Dynamic Mesh Actor Object Reference) -> (then, ReturnValue:Dynamic Mesh Pool Object Reference)
   > n5 = SplitMeshByMaterialIDs!(
    execute
    TargetMesh:Dynamic Mesh Object Reference=Get l_targetMesh() -> (l_targetMesh:Dynamic Mesh Object Reference)
    MeshPool:Dynamic Mesh Pool Object Reference=@n4
   ) -> (
    then
    ComponentMeshes:Array of Dynamic Mesh Object References
    ComponentMaterialIDs:Array of Integers
    ReturnValue:Dynamic Mesh Object Reference
   )
   > n6 = For Each Loop(Array:Array of Dynamic Mesh Object References=@n5) -> (Array Element:Dynamic Mesh Object Reference)
   n7 = "Get (a copy)"(Array:Array of Integers=@n5.ComponentMaterialIDs, Dimension 1=@n6.Array Index) -> (Output:int)
   n8 = Get l_comp() -> (l_comp:Static Mesh Component Object Reference)
   > n9 = ClearMaterialIDs!(
    execute
    TargetMesh:Dynamic Mesh Object Reference=@n6
    ClearValue:int=Find Item!(
     TargetArray:Array of Material Interface Object References=Get l_matArray() -> (l_matArray:Array of Material Interface Object References)
     ItemToFind:"Material Interface Object Reference (by ref)"=GetMaterial!(self:Primitive Component Object Reference=@n8, ElementIndex:int=@n7) -> (ReturnValue:Material Interface Object Reference)
    ) -> (ReturnValue)
   ) -> (then, ReturnValue:Dynamic Mesh Object Reference)
   > AppendMesh!(
    execute
    TargetMesh:Dynamic Mesh Object Reference=Get l_outputMesh() -> (l_outputMesh:Dynamic Mesh Object Reference)
    AppendMesh:Dynamic Mesh Object Reference=@n9
    AppendTransform:Transform
    bDeferChangeNotifications:bool=true
    AppendOptions:Geometry Script Append Mesh Options Structure=(CombineMode=EnableAllMatching)
   ) -> (then, ReturnValue:Dynamic Mesh Object Reference)
   Return Node!(execute=@n2.then_1, newMesh:Dynamic Mesh Object Reference=Get l_outputMesh())
   ClearMaterialIDs(TargetMesh=Get l_targetMesh(), ClearValue=0)
   n10 = Get l_smSource() -> (l_smSource:Static Mesh Object Reference)
   GetMaterial!(self:Static Mesh Object Reference=@n10, MaterialIndex:int=0) -> (ReturnValue)
   Get StaticMaterials!(self:Static Mesh Object Reference=@n10) -> (StaticMaterials:Array of Static Material Structures)
  ]}
 Create Mesh from Layout {nodes [
   n1 = Create Mesh from Layout!() -> (then, CurrentLayoutPackageName:String) {is=title K2Node_FunctionEntry}
   > Set CurrentLayoutPackageName(CurrentLayoutPackageName:String=@n1) -> (Output_Get:String)
   > n2 = Sequence()
   > n3 = Sequence!(execute) -> (then_0, then_1, then_2:Exec, then_3:Exec, then_4:Exec)
   > n4 = LoadAsset(AssetPath=Get CurrentLayoutPackageName() -> (CurrentLayoutPackageName:String))
   > n5 = GetBlueprintAsset(Object=@n4)
   > n6 = Gather Subobject Data For Blueprint(self=Get SubobjectDataSubsystem(), Context=@n5)
   > n7 = For Each Loop(Array:Array of Subobject Data Handle Structures=@n6) -> (Array Element:Subobject Data Handle Structure)
   > n8 = Cast To StaticMeshComponent(Object=GetObject!(Data:"Subobject Data Structure (by ref)"=GetData(DataHandle=@n7), bEvenIfPendingKill:bool=false) -> (ReturnValue:Object Reference))
   > Add(
    TargetArray:Array of Static Mesh Component Object References=Get l_comps() -> (l_comps:Array of Static Mesh Component Object References)
    NewItem:"Static Mesh Component Object Reference (by ref)"=@n8
   ) {is=title K2Node_CallArrayFunction}
   n9 = Get StaticMesh(self=@n8)
   > Add(
    TargetArray:Array of Static Mesh Object References=Get l_meshes() -> (l_meshes:Array of Static Mesh Object References)
    NewItem:"Static Mesh Object Reference (by ref)"=@n9
   )
   > Add(
    TargetArray:Array of Transforms=Get l_meshTransforms() -> (l_meshTransforms:Array of Transforms)
    NewItem:"Transform (by ref)"=GetRelativeTransform!(self:Scene Component Object Reference=@n8) -> (ReturnValue:Transform)
   )
   > Set l_minLODCount(l_minLODCount:int="Min (Integer)"(A=getValidLODCount!(mesh:Static Mesh Object Reference=@n9) -> (count:int=0), B=Get l_minLODCount() -> (l_minLODCount:int=0)) -> (ReturnValue=0)) -> (Output_Get:int=0)
   > Add Items!(
    execute
    TargetSet:Set of Material Interface Object References=Get l_matMap() -> (l_matMap:Set of Material Interface Object References)
    NewItems:Array of Material Interface Object References=GetMaterials!(self:Mesh Component Object Reference=@n8) -> (ReturnValue:Array of Material Interface Object References)
   ) -> (then)
   > n10 = GetSimpleCollisionFromComponent!(execute, Component:Primitive Component Object Reference=@n8) -> (then, ReturnValue:Geometry Script Simple Collision Structure)
   > n11 = GetSimpleCollisionShapeCount!(execute, SimpleCollision:"Geometry Script Simple Collision Structure (by ref)"=@n10) -> (then, ReturnValue:int=0)
   > Branch(Condition="integer > integer"(A=@n11))
   > Add(
    TargetArray:Array of Geometry Script Simple Collision Structures=Get l_simpleCollisionShapes() -> (l_simpleCollisionShapes:Array of Geometry Script Simple Collision Structures)
    NewItem:"Geometry Script Simple Collision Structure (by ref)"=@n10
   )
   For Loop(execute=@n3.then_1, LastIndex=int - int(A=Get l_minLODCount() -> (l_minLODCount=0)))
   > n12 = AllocateComputeMesh()
   > Add(
    TargetArray:Array of Dynamic Mesh Object References=Get l_LODs() -> (l_LODs:Array of Dynamic Mesh Object References)
    NewItem:"Dynamic Mesh Object Reference (by ref)"=@n12
   )
   n13 = For Each Loop(Exec=@n3.then_2, Array:Array of Static Mesh Object References=Get l_meshes()) -> (Array Element:Static Mesh Object Reference)
   > Set l_mesh(l_mesh:Static Mesh Object Reference=@n13) -> (Output_Get:Static Mesh Object Reference)
   > Set l_comp(l_comp="Get (a copy)"(Array:Array of Static Mesh Component Object References=Get l_comps(), Dimension 1=@n13.Array Index) -> (Output:Static Mesh Component Object Reference))
   > Set l_transform(l_transform:Transform="Get (a ref)"(Array:Array of Transforms=Get l_meshTransforms(), Dimension 1=@n13.Array Index) -> (Output:"Transform (by ref)")) -> (Output_Get:Transform)
   > n14 = For Loop(LastIndex=int - int(A=Get l_minLODCount() -> (l_minLODCount=0)))
   > Set l_lodIndex(l_lodIndex:int=@n14) -> (Output_Get:int=0)
   > n15 = AllocateComputeMesh()
   > n16 = Set l_dynamic_mesh(l_dynamic_mesh:Dynamic Mesh Object Reference=@n15) -> (Output_Get:Dynamic Mesh Object Reference)
   > n17 = Copy Mesh From Static Mesh with Section Materials!(
    execute
    FromStaticMeshAsset:Static Mesh Object Reference=Get l_mesh() -> (l_mesh:Static Mesh Object Reference)
    ToDynamicMesh:Dynamic Mesh Object Reference=@n16
    AssetOptions:Geometry Script Copy Mesh From Asset Options Structure=Make Geometry Script Copy Mesh from Asset Options!(bApplyBuildSettings:bool=false, bRequestTangents:bool=True, bIgnoreRemoveDegenerates:bool=True, bUseBuildScale:bool=True) -> (GeometryScriptCopyMeshFromAssetOptions:Geometry Script Copy Mesh From Asset Options Structure) {is=title K2Node_MakeStruct}
    RequestedLOD:Geometry Script Mesh Read LOD Structure=Make Geometry Script Mesh Read LOD!(LODType:EGeometryScriptLODType Enum=RenderData, LODIndex:int=Get l_lodIndex() -> (l_lodIndex:int=0)) -> (GeometryScriptMeshReadLOD:Geometry Script Mesh Read LOD Structure) {is=title K2Node_MakeStruct}
   ) -> (Failure:Exec, Success:Exec, ReturnValue:Dynamic Mesh Object Reference)
   n18 = To Array!(execute=@n17.Success, A:Set of Material Interface Object References=Get l_matMap()) -> (then, Result:Array of Material Interface Object References)
   > n19 = ApplyMaterialIDsToMesh!(
    execute
    self
    TargetMesh:Dynamic Mesh Object Reference=@n17
    MatArray:Array of Material Interface Object References=@n18
    Comp:Static Mesh Component Object Reference=Get l_comp()
   ) -> (then, newMesh:Dynamic Mesh Object Reference) {is=title K2Node_CallFunction}
   > n20 = TransformMesh!(
    execute
    TargetMesh:Dynamic Mesh Object Reference=@n19
    Transform:Transform=Get l_transform() -> (l_transform:Transform)
    bFixOrientationForNegativeScale:bool=true
   ) -> (then, ReturnValue:Dynamic Mesh Object Reference)
   n21 = BreakTransform(InTransform=Get l_transform())
   > n22 = ScaleMeshUVs!(
    execute
    TargetMesh:Dynamic Mesh Object Reference=@n20
    UVSetIndex:int=2
    Scale:Vector 2D Structure=MakeVector2D(X=1, Y=BreakVector(InVec=@n21.Scale).Y)
    ScaleOrigin:Vector 2D Structure=(X=0,Y=0.5)
    Selection:Mesh Selection Structure
   ) -> (then, ReturnValue:Dynamic Mesh Object Reference)
   > n23 = AppendMesh(
    TargetMesh="Get (a copy)"(Array:Array of Dynamic Mesh Object References=Get l_LODs(), Dimension 1=Get l_lodIndex() -> (l_lodIndex=0)) -> (Output:Dynamic Mesh Object Reference)
    AppendMesh=@n22
    bDeferChangeNotifications=false
    AppendOptions=Make Geometry Script Append Mesh Options!(CombineMode:EGeometryScriptCombineAttributesMode Enum=EnableAllMatching) -> (GeometryScriptAppendMeshOptions:Geometry Script Append Mesh Options Structure) {is=title K2Node_MakeStruct}
   )
   n24 = GetStaticMeshToSave!(execute=@n3.then_3, self, isForCollision:bool=false) -> (then, StaticMesh:Static Mesh Object Reference)
   > Set l_renderSM(l_renderSM:Static Mesh Object Reference=@n24) -> (Output_Get:Static Mesh Object Reference)
   > UpdateTrackMeshSettings!(
    execute
    self
    Mesh:Static Mesh Object Reference=Get l_renderSM() -> (l_renderSM:Static Mesh Object Reference)
    lodAmount:int=Get l_minLODCount() -> (l_minLODCount=0)
   ) -> (then)
   > n25 = To Array(A=Get l_matMap())
   > n26 = For Each Loop(Array:Array of Dynamic Mesh Object References=Get l_LODs()) -> (Array Element:Dynamic Mesh Object Reference)
   TransactObject(execute=@n26.Completed, Object=Get l_renderSM())
   n27 = PlanarProjectRoadUVs!(execute=@n26.LoopBody, self, TargetMesh:"Dynamic Mesh Object Reference (by ref)"=@n26) -> (then, TargetMeshOut:Dynamic Mesh Object Reference)
   > SaveComputeMeshToStaticMeshLOD!(
    execute
    self
    DynamicMesh:Dynamic Mesh Object Reference=@n27
    LODIndex:int=@n26.Array Index
    TargetStaticMeshAsset:Static Mesh Object Reference=Get l_renderSM()
    New Materials:Array of Material Interface Object References=@n25
   ) -> (then)
   n28 = To Array(execute=@n3.then_4, A=Get l_matMap())
   n29 = Get l_comps()
   > n30 = For Each Loop(Array:Array of Static Mesh Component Object References=@n29) -> (Array Element:Static Mesh Component Object Reference)
   n31 = GenerateCollisionMesh_Complex!(
    execute=@n30.Completed
    self
    components:Array of Static Mesh Component Object References=@n29
    materials:Array of Material Interface Object References=@n28
    l_simpleCollisions1:Array of Geometry Script Simple Collision Structures=Get l_simpleCollisionShapes()
   ) -> (then, collisionMesh:Dynamic Mesh Object Reference)
   > n32 = GetStaticMeshToSave(isForCollision=true)
   > PrintString!(
    execute
    InString="To String (Object)"!(InObj:Object Reference=@n32) -> (ReturnValue:String)
    bPrintToScreen
    bPrintToLog
    TextColor
    Duration
    Key
   ) -> (then)
   > n33 = Set l_collisionSM(l_collisionSM:Static Mesh Object Reference=@n32) -> (Output_Get:Static Mesh Object Reference)
   > PrintString(InString="To String (Object)"(InObj=@n33))
   > SaveComputeMeshToStaticMeshCollision!(
    execute
    self
    CollisionDynamicMesh:Dynamic Mesh Object Reference=@n31
    CollisionStaticMeshAsset:Static Mesh Object Reference=@n33
    New Materials:Array of Material Interface Object References=@n28
    renderStaticMesh:Static Mesh Object Reference=Get l_renderSM()
    SimpleCollisions:Array of Geometry Script Simple Collision Structures=Get l_simpleCollisionShapes()
   ) -> (then)
   > SetEditorProperty(
    Object=Get l_renderSM()
    PropertyName=ComplexCollisionMesh
    PropertyValue:Static Mesh Object Reference=Get l_collisionSM() -> (l_collisionSM:Static Mesh Object Reference)
   )
   > n34 = GetEditorProperty!(execute:Exec, Object:Object Reference=Get l_renderSM(), PropertyName:Name=BodySetup) -> (then:Exec, PropertyValue:Object Reference, ReturnValue:bool)
   > SetEditorProperty(
    Object=@n34
    PropertyName=CollisionTraceFlag
    PropertyValue:ECollisionTraceFlag Enum=Literal enum ECollisionTraceFlag!(Enum:ECollisionTraceFlag Enum=CTF_UseComplexAsSimple) -> (ReturnValue:ECollisionTraceFlag Enum) {is=title K2Node_EnumLiteral}
    ChangeNotifyMode=Always
   )
   > n35 = GetEditorProperty(Object=Get l_collisionSM(), PropertyName=BodySetup)
   > SetEditorProperty(
    Object=@n35
    PropertyName=CollisionTraceFlag
    PropertyValue=Literal enum ECollisionTraceFlag(Enum=CTF_UseComplexAsSimple)
    ChangeNotifyMode=Always
   )
   PrintString(execute=@n30.LoopBody, InString="To String (Object)"(InObj=@n30), bPrintToScreen=false)
   ReleaseAllComputeMeshes(execute=@n2.then_1, self:Dynamic Mesh Actor Object Reference)
   PlanarProjectRoadUVs(TargetMesh=@n23)
  ]}
 EventGraph {nodes [
   n1 = Event OnRebuildGeneratedMesh!() -> (OutputDelegate, then, TargetMesh:Dynamic Mesh Object Reference) {is=title K2Node_Event}
   > Set CurrentTargetMesh(CurrentTargetMesh:Dynamic Mesh Object Reference=@n1.TargetMesh) -> (Output_Get:Dynamic Mesh Object Reference)
  ]}
 GenerateCollisionMesh_Complex {nodes [
   n1 = GenerateCollisionMesh_Complex!() -> (
    then
    components:Array of Static Mesh Component Object References
    materials:Array of Material Interface Object References
    l_simpleCollisions1:Array of Geometry Script Simple Collision Structures
   ) {is=title K2Node_FunctionEntry}
   > Set l_comps(l_comps:Array of Static Mesh Component Object References=@n1) -> (Output_Get:Array of Static Mesh Component Object References)
   > Set l_mats(l_mats:Array of Material Interface Object References=@n1.materials) -> (Output_Get:Array of Material Interface Object References)
   > Set l_simpleCollisions(l_simpleCollisions:Array of Geometry Script Simple Collision Structures=@n1.l_simpleCollisions1) -> (Output_Get:Array of Geometry Script Simple Collision Structures)
   > n2 = AllocateComputeMesh()
   > Set l_computeOutput(l_computeOutput:Dynamic Mesh Object Reference=@n2) -> (Output_Get:Dynamic Mesh Object Reference)
   > n3 = Sequence!(execute) -> (then_0, then_1)
   > n4 = For Each Loop(Array=Get l_comps())
   > n5 = Set l_comp(l_comp=@n4)
   > Branch(Condition=Is Collision Enabled!(self:Primitive Component Object Reference=@n5) -> (ReturnValue:bool))
   > n6 = AllocateComputeMesh()
   > Set l_computeReuse(l_computeReuse:Dynamic Mesh Object Reference=@n6) -> (Output_Get:Dynamic Mesh Object Reference)
   n7 = Get StaticMesh(self=Get l_comp())
   > n8 = Copy Mesh From Static Mesh with Section Materials(
    FromStaticMeshAsset=Get StaticMesh(self=Get l_comp())
    ToDynamicMesh=Get l_computeReuse() -> (l_computeReuse:Dynamic Mesh Object Reference)
    AssetOptions=Make Geometry Script Copy Mesh from Asset Options(bApplyBuildSettings=false, bRequestTangents=false, bIgnoreRemoveDegenerates=True, bUseBuildScale=True)
    RequestedLOD=Make Geometry Script Mesh Read LOD(LODType=RenderData, LODIndex=Get LODForCollision!(self:Static Mesh Object Reference=@n7) -> (LODForCollision:int=0))
   )
   n9 = ApplyMaterialIDsToMesh(execute=@n8.Success, TargetMesh=@n8, MatArray=Get l_mats() -> (l_mats:Array of Material Interface Object References), Comp=Get l_comp())
   > Set l_computeReuse(l_computeReuse=@n9)
   > n10 = TransformMesh(TargetMesh=Get l_computeReuse(), Transform=GetRelativeTransform(self=Get l_comp()), bFixOrientationForNegativeScale=true)
   > n11 = AppendMesh(
    TargetMesh=Get l_computeOutput() -> (l_computeOutput:Dynamic Mesh Object Reference)
    AppendMesh=@n10
    bDeferChangeNotifications=false
    AppendOptions=Make Geometry Script Append Mesh Options(CombineMode=EnableAllMatching)
   )
   > DiscardTangents!(execute, TargetMesh:Dynamic Mesh Object Reference=@n11) -> (then, ReturnValue:Dynamic Mesh Object Reference)
   Return Node!(execute=@n3.then_1, collisionMesh:Dynamic Mesh Object Reference=Get l_computeOutput())
  ]}
 GetLayoutSettingsOverride {nodes [
   n1 = LoadAsset(AssetPath=Get CurrentLayoutPackageName())
   n2 = CreateAssetData!(InAsset=@n1, bAllowBlueprintClass=false) -> (
    ReturnValue_PackageName:Name
    ReturnValue_PackagePath:Name
    ReturnValue_AssetName:Name
    ReturnValue_AssetClassPath:Top Level Asset Path Structure
   )
   > n3 = Execute Python Script!(
    execute
    PythonScript="import unreal\r\nasset_class = unreal.load_class(None, in_class_string)\r\ncdo = unreal.get_default_object(asset_class)"
    in_class_string:String=Append!(A="To String (Name)"(InName=@n2), B=., C:String="To String (Name)"(InName=@n2.ReturnValue_AssetName), D:String=_C) -> (ReturnValue)
   ) -> (then, ReturnValue, cdo:Object Reference) {inputs [in_class_string], outputs [cdo]}
   > n4 = GetEditorProperty(Object=@n3.cdo, PropertyName=UseOverrideSettings) -> (PropertyValue:bool)
   > n5 = Branch(Condition=@n4)
   > n6 = GetEditorProperty(Object=@n3.cdo, PropertyName=SettingsOverride) -> (PropertyValue:"Structure (by ref)")
   > Return Node!(execute, Settings:"Structure (by ref)"=@n6)
   Return Node(execute=[@n7.then, @n5.else], Settings="Make <unknown struct>"!() -> (TrackCombinerSettings:Structure) {is=title K2Node_MakeStruct})
   n7 = GetLayoutSettingsOverride() {is=title K2Node_FunctionEntry}
  ]}
 GetSetting_LODSizes {nodes [
   GetSetting_LODSizes() {is=title K2Node_FunctionEntry}
   > n1 = GetLayoutSettingsOverride!(execute, self) -> (then, Settings:Structure) {is=title K2Node_CallFunction}
   > Return Node!(
    execute
    LODScreenSizes:"Array of Float (single-precision)s"="Break <unknown struct>"!(TrackCombinerSettings:"Structure (by ref)"=@n1) -> (LODScreenSizes_3_#1:"Array of Float (double-precision)s") {is=title K2Node_BreakStruct}
   )
  ]}
 GetStaticMeshToSave {nodes [
   n1 = GetStaticMeshToSave!() -> (then, isForCollision:bool=~) {is=title K2Node_FunctionEntry}
   > Set l_forCollision(l_forCollision:bool=@n1) -> (Output_Get:bool)
   n2 = Split(SourceString=Get CurrentLayoutPackageName(), InStr=/, SearchDir=FromEnd)
   > n3 = GetLayoutOrMeshNamesFromOther!(
    execute
    assetName:String=@n2.RightS
    InputIsMesh:bool=false
    isTesting:bool=Get isTesting() -> (isTesting:bool)
    isCollision:bool=Get l_forCollision() -> (l_forCollision:bool)
    SavePathRoot:String=Append!(A=Split(SourceString=@n2, InStr=/, SearchDir=FromEnd), B=/) -> (ReturnValue)
   ) -> (then, AssetExists:bool, PackageName:String)
   > n4 = Set l_FinalSavePackageName(l_FinalSavePackageName:String=@n3.PackageName) -> (Output_Get:String)
   > n5 = DoesAssetExist(AssetPath=@n4)
   > n6 = Branch(Condition=@n5)
   > n7 = LoadAsset(AssetPath=Get l_FinalSavePackageName() -> (l_FinalSavePackageName:String))
   > n8 = Cast To StaticMesh(Object=@n7)
   > Return Node!(execute, StaticMesh:Static Mesh Object Reference=@n8)
   n9 = GetAssetTools(execute=@n6.else)
   n10 = Split(SourceString=Get l_FinalSavePackageName(), InStr=/, SearchDir=FromEnd)
   > n11 = CreateAsset(self=@n9, AssetName=@n10.RightS, PackagePath=@n10, AssetClass=&StaticMesh)
   > n12 = Cast To StaticMesh(Object=@n11)
   > SetGenerateLightmapUVs!(
    execute
    self:Static Mesh Editor Subsystem Object Reference=Get StaticMeshEditorSubsystem()
    StaticMesh:Static Mesh Object Reference=@n12
    bGenerateLightmapUVs:bool=false
   ) -> (then, ReturnValue:bool)
   > Return Node(StaticMesh=@n12)
   Reroute Node(InputPin:Wildcard) -> (OutputPin:Wildcard)
  ]}
 PlanarProjectRoadUVs {nodes [
   n1 = Break Box!(Box:"Box Structure (by ref)"=GetMeshBoundingBox!(TargetMesh:Dynamic Mesh Object Reference=Get l_mesh() -> (l_mesh:Dynamic Mesh Object Reference)) -> (ReturnValue:Box Structure)) -> (Min:Vector=~, Max:Vector=~, IsValid:bool=~) {is=title K2Node_BreakStruct}
   n2 = BreakVector(InVec=@n1.Max)
   n3 = BreakVector(InVec=@n1)
   n4 = "Max (Float)"(A="Absolute (Float)"(A=@n3.Y), B=@n2.Y)
   n5 = MakeVector(X=@n2, Y=@n4, Z=@n2.Z)
   n6 = BreakVector(InVec=@n5)
   n7 = MakeVector(X=@n6.Y, Y=@n6.Z)
   n8 = BreakVector(InVec=MakeVector(X=@n3, Y=float * float(A=@n4, B=-1) {is=title K2Node_PromotableOperator}, Z=@n3.Z))
   n9 = MakeVector(X=@n8.Y, Y=@n8.Z)
   n10 = PlanarProjectRoadUVs!() -> (then, TargetMesh:"Dynamic Mesh Object Reference (by ref)") {is=title K2Node_FunctionEntry}
   > n11 = Set l_mesh(l_mesh:Dynamic Mesh Object Reference=@n10) -> (Output_Get:Dynamic Mesh Object Reference)
   > n12 = SetMeshUVsFromPlanarProjection!(
    execute
    TargetMesh:Dynamic Mesh Object Reference=@n11
    UVSetIndex:int=1
    PlaneTransform:Transform=MakeTransform(Location=@n5, Rotation="0,90,90", Scale=vector - vector(A=@n7, B=@n9))
    Selection:Mesh Selection Structure
   ) -> (then, ReturnValue:Dynamic Mesh Object Reference)
   > n13 = TranslateMeshUVs!(
    execute
    TargetMesh:Dynamic Mesh Object Reference=@n12
    UVSetIndex:int=1
    Translation:Vector 2D Structure=(X=1,Y=0)
    Selection:Mesh Selection Structure
   ) -> (then, ReturnValue:Dynamic Mesh Object Reference)
   > n14 = ComputeTangents!(
    execute
    TargetMesh:Dynamic Mesh Object Reference=@n13
    Options:Geometry Script Tangents Options Structure=Make Geometry Script Tangents Options!(Type:EGeometryScriptTangentTypes Enum=FastMikkT, UVLayer:int=2) -> (GeometryScriptTangentsOptions:Geometry Script Tangents Options Structure) {is=title K2Node_MakeStruct}
   ) -> (then, ReturnValue:Dynamic Mesh Object Reference)
   n15 = RecomputeNormals!(
    execute
    TargetMesh:Dynamic Mesh Object Reference=@n14
    CalculateOptions:Geometry Script Calculate Normals Options Structure=Make Geometry Script Calculate Normals Options!(bAngleWeighted:bool=True, bAreaWeighted:bool=True) -> (GeometryScriptCalculateNormalsOptions:Geometry Script Calculate Normals Options Structure) {is=title K2Node_MakeStruct}
    bDeferChangeNotifications:bool=false
   ) -> (then, ReturnValue:Dynamic Mesh Object Reference)
   Return Node!(execute=[@n14.then, @n15.then], TargetMeshOut:Dynamic Mesh Object Reference=@n14)
   Make Geometry Script Split Normals Options!(
    bSplitByOpeningAngle:bool=True
    OpeningAngleDeg:float=15
    bSplitByFaceGroup:bool=False
    GroupLayer:Geometry Script Group Layer Structure=(bDefaultLayer=True,ExtendedLayerIndex=0)
   ) -> (GeometryScriptSplitNormalsOptions:Geometry Script Split Normals Options Structure) {is=title K2Node_MakeStruct}
  ]}
 RebuildAllLayouts {nodes [
   RebuildAllLayouts() {is=title K2Node_FunctionEntry}
   > ResetTool()
   > n1 = For Each Loop(Array:Array of Strings=Get LayoutsToBuild() -> (LayoutsToBuild:Array of Strings)) -> (Array Element:String)
   Branch(execute=@n1.Completed, Condition=Get KillOnComplete() -> (KillOnComplete:bool))
   > DestroyActor!(execute, self:Editor Actor Subsystem Object Reference=Get EditorActorSubsystem(), ActorToDestroy:Actor Object Reference=Self-Reference()) -> (then, ReturnValue:bool)
   Create Mesh from Layout!(execute=@n1.LoopBody, self, CurrentLayoutPackageName:String=@n1) -> (then) {is=title K2Node_CallFunction}
  ]}
 ResetTool {nodes [
   ResetTool() {is=title K2Node_FunctionEntry}
   > Clear(TargetArray:Array of Object References=Get AssetsToOpenPostBuild() -> (AssetsToOpenPostBuild:Array of Object References))
  ]}
 SaveComputeMeshToStaticMeshCollision {nodes [
   n1 = SetStaticMeshCollisionFromMesh!(
    execute
    FromDynamicMesh:Dynamic Mesh Object Reference=Get l_colMesh() -> (l_colMesh:Dynamic Mesh Object Reference)
    ToStaticMeshAsset:Static Mesh Object Reference=Get l_ColSM() -> (l_ColSM:Static Mesh Object Reference)
    Options:Geometry Script Collision From Mesh Options Structure
    StaticMeshCollisionOptions:Geometry Script Set Static Mesh Collision Options Structure=(bMarkAsCustomized=True)
   ) -> (then, ReturnValue:Dynamic Mesh Object Reference)
   RemoveCollisions!(
    execute=[@n6.Success, @n1.then]
    self:Static Mesh Editor Subsystem Object Reference=Get StaticMeshEditorSubsystem()
    StaticMesh:Static Mesh Object Reference=Get l_renderStaticMesh() -> (l_renderStaticMesh:Static Mesh Object Reference)
   ) -> (then, ReturnValue:bool)
   n2 = Make Static Mesh Reduction Settings!(PercentTriangles:float=1, ScreenSize:float=1) -> (StaticMeshReductionSettings:Static Mesh Reduction Settings Structure) {is=title K2Node_MakeStruct}
   > SetLods!(
    execute
    self:Static Mesh Editor Subsystem Object Reference=Get StaticMeshEditorSubsystem()
    StaticMesh:Static Mesh Object Reference=Get l_ColSM()
    ReductionOptions:"Static Mesh Reduction Options Structure (by ref)"=Make Static Mesh Reduction Options!(
     bAutoComputeLODScreenSize:bool=false
     ReductionSettings:Array of Static Mesh Reduction Settings Structures=Make Array("[0]":Static Mesh Reduction Settings Structure=@n2) -> (Array:Array of Static Mesh Reduction Settings Structures)
    ) -> (StaticMeshReductionOptions:Static Mesh Reduction Options Structure) {is=title K2Node_MakeStruct}
   ) -> (then, ReturnValue:int=0)
   n3 = SaveComputeMeshToStaticMeshCollision!() -> (
    then
    CollisionDynamicMesh:Dynamic Mesh Object Reference
    CollisionStaticMeshAsset:Static Mesh Object Reference
    New Materials:Array of Material Interface Object References
    renderStaticMesh:Static Mesh Object Reference
    SimpleCollisions:Array of Geometry Script Simple Collision Structures
   ) {is=title K2Node_FunctionEntry}
   > Set l_colMesh(l_colMesh:Dynamic Mesh Object Reference=@n3) -> (Output_Get:Dynamic Mesh Object Reference)
   > Set l_ColSM(l_ColSM:Static Mesh Object Reference=@n3.CollisionStaticMeshAsset) -> (Output_Get:Static Mesh Object Reference)
   > Set l_renderStaticMesh(l_renderStaticMesh:Static Mesh Object Reference=@n3.renderStaticMesh) -> (Output_Get:Static Mesh Object Reference)
   > Set l_simpleCollision(l_simpleCollision:Array of Geometry Script Simple Collision Structures=@n3.SimpleCollisions) -> (Output_Get:Array of Geometry Script Simple Collision Structures)
   > n4 = Sequence!(execute) -> (then_0, then_1, then_2:Exec)
   > n5 = CloseAllEditorsForAsset!(execute, self:Asset Editor Subsystem Object Reference=Get AssetEditorSubsystem(), Asset:Object Reference=Get l_ColSM()) -> (then, ReturnValue:int=0)
   > Branch(Condition="integer > integer"(A=@n5))
   > Add(TargetArray:Array of Object References=Get AssetsToOpenPostBuild(), NewItem:"Object Reference (by ref)"=Get l_ColSM())
   n6 = CopyMeshToStaticMesh!(
    execute=@n4.then_2
    FromDynamicMesh:Dynamic Mesh Object Reference=Get l_colMesh()
    ToStaticMeshAsset:Static Mesh Object Reference=Get l_ColSM()
    Options:Geometry Script Copy Mesh To Asset Options Structure=Make Geometry Script Copy Mesh to Asset Options!(
     bEnableRecomputeNormals:bool=False
     bEnableRecomputeTangents:bool=False
     bEnableRemoveDegenerates:bool=False
     BoneHierarchyMismatchHandling:EGeometryScriptBoneHierarchyMismatchHandling Enum=DoNothing
     bRemapBoneIndicesToMatchAsset:bool=False
     bUseOriginalVertexOrder:bool=False
     bUseBuildScale:bool=True
     bReplaceMaterials:bool=true
     NewMaterialUsage:ECopyMeshToAssetUseNewMaterialsOptions Enum=ReplaceAssetMaterials
     NewMaterials:Array of Material Interface Object References=@n3.New Materials
     NewMaterialSlotNames:Array of Names
     bCleanAssignedMaterials:bool=False
     GenerateLightmapUVs:EGeometryScriptGenerateLightmapUVOptions Enum=MatchTargetLODSetting
     bApplyNaniteSettings:bool=False
     NaniteSettings:Geometry Script Nanite Options Structure=(bEnabled=True,FallbackPercentTriangles=100,FallbackRelativeError=0)
     NewNaniteSettings:Mesh Nanite Settings Structure=(bEnabled=False,bExplicitTangents=False,bLerpUVs=True,bSeparable=False,bVoxelNDF=True,bVoxelOpacity=False,ShapePreservation=None,PositionPrecision=-2147483648,NormalPrecision=-1,TangentPrecision=-1,BoneWeightPrecision=-1,TargetMinimumResidencyInKB=0,KeepPercentTriangles=1,TrimRelativeError=0,GenerateFallback=PlatformDefault,FallbackTarget=Auto,FallbackPercentTriangles=1,FallbackRelativeError=1,MaxEdgeLengthFactor=0,NumRays=64,VoxelLevel=0,RayBackUp=0,DisplacementUVChannel=0,DisplacementMaps=,NaniteAssemblyData=(Parts=,Nodes=))
     bEmitTransaction:bool=false
     bDeferMeshPostEditChange:bool=False
    ) -> (GeometryScriptCopyMeshToAssetOptions:Geometry Script Copy Mesh To Asset Options Structure) {is=title K2Node_MakeStruct}
    TargetLOD:Geometry Script Mesh Write LOD Structure=Make Geometry Script Mesh Write LOD!(bWriteHiResSource:bool=False, LODIndex:int=0) -> (GeometryScriptMeshWriteLOD:Geometry Script Mesh Write LOD Structure) {is=title K2Node_MakeStruct}
    bUseSectionMaterials:bool=true
   ) -> (Failure:Exec, Success:Exec, ReturnValue:Dynamic Mesh Object Reference)
   SetSimpleCollisionOfStaticMesh!(
    execute
    SimpleCollision:"Geometry Script Simple Collision Structure (by ref)"
    StaticMesh:Static Mesh Object Reference=Get l_ColSM()
    Options:Geometry Script Set Simple Collision Options Structure
    StaticMeshCollisionOptions:Geometry Script Set Static Mesh Collision Options Structure=(bMarkAsCustomized=True)
   ) -> (then)
   For Each Loop(Array:Array of Geometry Script Simple Collision Structures=Get l_simpleCollision() -> (l_simpleCollision:Array of Geometry Script Simple Collision Structures)) -> (Array Element:Geometry Script Simple Collision Structure)
  ]}
 SaveComputeMeshToStaticMeshLOD {nodes [
   n1 = SaveComputeMeshToStaticMeshLOD!() -> (
    then
    DynamicMesh:Dynamic Mesh Object Reference
    LODIndex:int
    TargetStaticMeshAsset:Static Mesh Object Reference
    New Materials:Array of Material Interface Object References
   ) {is=title K2Node_FunctionEntry}
   > Set l_dynMesh(l_dynMesh:Dynamic Mesh Object Reference=@n1) -> (Output_Get:Dynamic Mesh Object Reference)
   > Set l_meshToSave(l_meshToSave:Static Mesh Object Reference=@n1.TargetStaticMeshAsset) -> (Output_Get:Static Mesh Object Reference)
   > n2 = Sequence()
   > n3 = CloseAllEditorsForAsset(self=Get AssetEditorSubsystem(), Asset=Get l_meshToSave() -> (l_meshToSave:Static Mesh Object Reference)) -> (ReturnValue=0)
   > Branch(Condition="integer > integer"(A=@n3))
   > Add(TargetArray=Get AssetsToOpenPostBuild(), NewItem=Get l_meshToSave())
   n4 = CopyMeshToStaticMesh(
    execute=@n2.then_2
    FromDynamicMesh=Get l_dynMesh() -> (l_dynMesh:Dynamic Mesh Object Reference)
    ToStaticMeshAsset=Get l_meshToSave()
    Options=Make Geometry Script Copy Mesh to Asset Options(
     bEnableRecomputeNormals=False
     bEnableRecomputeTangents=False
     bEnableRemoveDegenerates=true
     BoneHierarchyMismatchHandling=DoNothing
     bRemapBoneIndicesToMatchAsset=False
     bUseOriginalVertexOrder=False
     bUseBuildScale=True
     bReplaceMaterials=true
     NewMaterialUsage=ReplaceAssetMaterials
     NewMaterials=@n1.New Materials
     bCleanAssignedMaterials=False
     GenerateLightmapUVs=MatchTargetLODSetting
     bApplyNaniteSettings=true
     NaniteSettings=(bEnabled=True,FallbackPercentTriangles=100,FallbackRelativeError=0)
     NewNaniteSettings=Make Mesh Nanite Settings!(
      bEnabled:bool=true
      bExplicitTangents:bool=true
      bLerpUVs:bool=True
      bSeparable:bool=False
      bVoxelNDF:bool=True
      bVoxelOpacity:bool=False
      ShapePreservation:ENaniteShapePreservation Enum=None
      PositionPrecision:int=-2147483648
      NormalPrecision:int=-1
      TangentPrecision:int=-1
      BoneWeightPrecision:int=-1
      KeepPercentTriangles:float=1
      TrimRelativeError:float=0
      GenerateFallback:ENaniteGenerateFallback Enum=PlatformDefault
      FallbackTarget:ENaniteFallbackTarget Enum=PercentTriangles
      FallbackPercentTriangles:float=100
      FallbackRelativeError:float=1
      MaxEdgeLengthFactor:float=64
      NumRays:int=64
      VoxelLevel:int=0
      RayBackUp:float=0
      DisplacementUVChannel:int=0
     ) -> (MeshNaniteSettings:Mesh Nanite Settings Structure) {is=title K2Node_MakeStruct}
     bEmitTransaction=True
     bDeferMeshPostEditChange=False
    )
    TargetLOD=Make Geometry Script Mesh Write LOD(bWriteHiResSource=False, LODIndex=@n1.LODIndex)
    bUseSectionMaterials=true
   )
   n5 = Get StaticMeshEditorSubsystem()
   n6 = Get l_meshToSave()
   SetGenerateLightmapUVs(execute=@n4.Success, self=@n5, StaticMesh=@n6, bGenerateLightmapUVs=false)
   > RemoveUVChannel!(
    execute
    self:Static Mesh Editor Subsystem Object Reference=@n5
    StaticMesh:Static Mesh Object Reference=@n6
    LODIndex:int=@n1.LODIndex
    UVChannelIndex:int=3
   ) -> (then, ReturnValue:bool)
  ]}
 SetLayoutsToBuild {nodes [
   n1 = SetLayoutsToBuild!() -> (then, TargetArray:Array of Strings) {is=title K2Node_FunctionEntry}
   > Add Items(TargetSet:Set of Strings=Get l_set() -> (l_set:Set of Strings), NewItems:Array of Strings=@n1)
   > n2 = To Array(A:Set of Strings=Get l_set()) -> (Result:Array of Strings)
   > n3 = Set LayoutsToBuild(LayoutsToBuild:Array of Strings=@n2) -> (Output_Get:Array of Strings)
   > Branch(Condition=Is Valid Index!(TargetArray=@n3, IndexToTest:int) -> (ReturnValue:bool))
   > RebuildAllLayouts() {is=title K2Node_CallFunction}
   > OpenEditorForAssets!(
    execute
    self:Asset Editor Subsystem Object Reference=Get AssetEditorSubsystem()
    Assets:Array of Object References=Get AssetsToOpenPostBuild()
    OpenedMethod:EAssetTypeActivationOpenedMethod Enum=Edit
   ) -> (then, ReturnValue:bool)
  ]}
 SetMeshNamesToBuild {nodes [
   n1 = SetMeshNamesToBuild!() -> (then, TargetArray:Array of Strings, AssetPathRoot:String) {is=title K2Node_FunctionEntry}
   > Add Items(TargetSet=Get l_set(), NewItems=@n1)
   > n2 = To Array(A=Get l_set())
   > n3 = For Each Loop(Array:Array of Strings=@n2) -> (Array Element:String)
   SetLayoutsToBuild!(execute=@n3.Completed, self, TargetArray:Array of Strings=Get l_layouts() -> (l_layouts:Array of Strings)) -> (then) {is=title K2Node_CallFunction}
   n4 = Len(S=@n1.AssetPathRoot)
   n5 = getLayoutNameFromMeshName!(
    execute=@n3.LoopBody
    self
    MeshAssetName:String=@n3
    AssetPathRoot:String=SelectString(A=@n1.AssetPathRoot, B=/DelMarGame/Track/, bPickA="integer > integer"(A=@n4))
   ) -> (then, layoutExists:bool, LayoutPackageName:String)
   > Branch(Condition=@n5)
   > Add(TargetArray:Array of Strings=Get l_layouts(), NewItem:"String (by ref)"=@n5.LayoutPackageName)
  ]}
 UpdateSettingsOnTracks {nodes [
   n1 = UpdateSettingsOnTracks!() -> (then, TrackMeshes:Array of Static Mesh Object References) {is=title K2Node_FunctionEntry}
   > n2 = For Each Loop(Array:Array of Static Mesh Object References=@n1) -> (Array Element:Static Mesh Object Reference)
   n3 = GetSoftObjectPath!(Object=@n2) -> (ReturnValue_PathString:String)
   > n4 = GetLayoutOrMeshNamesFromOther(
    assetName=Split(SourceString=Split(SourceString=@n3, InStr=., SearchDir=FromStart), InStr=/, SearchDir=FromEnd).RightS
    InputIsMesh=true
    isTesting=Get isTesting()
    isCollision=false
   )
   > Branch(Condition=@n4)
   > Set CurrentLayoutPackageName(CurrentLayoutPackageName=@n4.PackageName)
   > UpdateTrackMeshSettings(Mesh=@n2, lodAmount=0)
  ]}
 UpdateTrackMeshSettings {nodes [
   n1 = UpdateTrackMeshSettings!() -> (then, Mesh:Static Mesh Object Reference, lodAmount:int) {is=title K2Node_FunctionEntry}
   > Set l_mesh(l_mesh:Static Mesh Object Reference=@n1) -> (Output_Get:Static Mesh Object Reference)
   > Set l_count(l_count:int=@n1.lodAmount) -> (Output_Get:int=0)
   > n2 = Sequence()
   > Set l_screenSizes(l_screenSizes:"Array of Float (single-precision)s"=GetSetting_LODSizes!() -> (LODScreenSizes:"Array of Float (single-precision)s") {is=title K2Node_CallFunction}) -> (Output_Get:"Array of Float (single-precision)s")
   n3 = Get l_screenSizes() -> (l_screenSizes:"Array of Float (single-precision)s")
   > Branch(Condition="integer > integer"(A=Length(TargetArray:"Array of Float (single-precision)s"=@n3), B=Get l_count() -> (l_count:int=0)))
   > Resize!(execute, TargetArray:"Array of Float (single-precision)s"=Get l_screenSizes(), Size:int=Get l_count() -> (l_count=0)) -> (then) {is=title K2Node_CallArrayFunction}
   n4 = For Each Loop(Exec=@n2.then_1, Array:"Array of Float (single-precision)s"=Get l_screenSizes()) -> (Array Element:float)
   > Add!(
    execute
    TargetArray:Array of Static Mesh Reduction Settings Structures=Get l_settings() -> (l_settings:Array of Static Mesh Reduction Settings Structures)
    NewItem_PercentTriangles:float=1
    NewItem_ScreenSize:float=@n4
   ) -> (then, ReturnValue)
   Set LODForCollision!(execute=@n2.then_2, LODForCollision:int=0, self:Static Mesh Object Reference=Get l_mesh() -> (l_mesh:Static Mesh Object Reference)) -> (then, Output_Get:int=0)
   > SetLods(
    self=Get StaticMeshEditorSubsystem()
    StaticMesh=Get l_mesh()
    ReductionOptions=Make Static Mesh Reduction Options(bAutoComputeLODScreenSize=false, ReductionSettings=Get l_settings())
   ) -> (ReturnValue=0)
   TransactObject(Object=Get l_mesh())
   SetLodScreenSizes!(
    execute
    self:Static Mesh Editor Subsystem Object Reference=Get StaticMeshEditorSubsystem()
    StaticMesh:Static Mesh Object Reference=Get l_mesh()
    ScreenSizes:"Array of Float (single-precision)s"=Get l_screenSizes()
   ) -> (then, ReturnValue:bool)
  ]}
 UserConstructionScript {nodes [
   Construction Script()
   > n1 = Branch(Condition=Get RunOnSelectedLayouts() -> (RunOnSelectedLayouts:bool))
   > Set RunOnSelectedLayouts(RunOnSelectedLayouts:bool=false) -> (Output_Get:bool)
   > n2 = GetSelectedAssetData()
   > n3 = FilterAssetByParent!(
    execute
    ClassName:Name=TrackLayoutActor
    bUseNativeParent:bool=false
    InAssetData:Array of Asset Data Structures=@n2
    bExactMatch:bool=false
   ) -> (then, OutAssetData:Array of Asset Data Structures)
   > n4 = For Each Loop(Array:Array of Asset Data Structures=@n3) -> (Array Element:Asset Data Structure)
   SetLayoutsToBuild(execute=@n4.Completed, TargetArray=Get l_layoutPackages() -> (l_layoutPackages:Array of Strings))
   Add!(
    execute=@n4.LoopBody
    TargetArray:Array of Strings=Get l_layoutPackages()
    NewItem:"String (by ref)"="To String (Name)"(InName=Break Asset Data(AssetData=@n4))
   ) -> (then, ReturnValue)
   Branch(execute=@n1.else, Condition=Get RunOnTestLayout() -> (RunOnTestLayout:bool))
   > Set RunOnTestLayout(RunOnTestLayout:bool=false) -> (Output_Get:bool)
   > SetLayoutsToBuild(TargetArray=Make Array("[0]":String=/DelMarGame/Track/Layouts/Layout_DSRT_Basic_GR) -> (Array:Array of Strings))
  ]}
 VirtualMIrror {nodes [
   n1 = Inputs!() -> (MinA:double=~, MinB:double=~, MaxA:double=~, MaxB:double=~)
   n2 = "Min (Float)"(A:double=@n1.MaxA, B:double=@n1) -> (ReturnValue:double) {is=title K2Node_CommutativeAssociativeBinaryOperator}
   Outputs!(MaxWidth:double="Max (Float)"(A="Max (Float)"(A=@n1.MaxB, B=@n1.MinB), B="Absolute (Float)"(A=@n2)))
  ]}
 getLayoutNameFromMesh {nodes [
   n1 = getLayoutNameFromMesh!() -> (then, staticMesh:Static Mesh Object Reference) {is=title K2Node_FunctionEntry}
   n2 = GetSoftObjectPath(Object=@n1)
   > n3 = GetLayoutOrMeshNamesFromOther(
    assetName=Split(SourceString=Split(SourceString=@n2, InStr=., SearchDir=FromStart), InStr=/, SearchDir=FromStart).RightS
    InputIsMesh=true
    isTesting=Get isTesting()
    isCollision=false
   )
   > Return Node!(execute, layoutExists:bool=@n3, LayoutPackageName:String=@n3.PackageName)
  ]}
 getLayoutNameFromMeshName {nodes [
   n1 = getLayoutNameFromMeshName!() -> (then, MeshAssetName:String, AssetPathRoot:String) {is=title K2Node_FunctionEntry}
   > n2 = Set l_meshAssetName(l_meshAssetName:String=@n1) -> (Output_Get:String)
   > n3 = Set l_meshAssetName(l_meshAssetName=Replace(SourceString=@n2, From=Get SavePrefix_Mesh() -> (SavePrefix_Mesh:String), To=Get SavePrefix_Layout() -> (SavePrefix_Layout:String)))
   > n4 = Set l_layoutPackageName(l_layoutPackageName:String=Append(A=Append(A=@n1.AssetPathRoot, B=Layouts/), B=@n3)) -> (Output_Get:String)
   > n5 = DoesAssetExist(AssetPath=@n4)
   > Return Node(layoutExists=@n5, LayoutPackageName=Get l_layoutPackageName() -> (l_layoutPackageName:String))
  ]}
 getValidLODCount {nodes [
   n1 = getValidLODCount!() -> (then, mesh:Static Mesh Object Reference) {is=title K2Node_FunctionEntry}
   > n2 = For Each Loop(Array:"Array of Float (single-precision)s"=GetLodScreenSizes!(self:Static Mesh Editor Subsystem Object Reference=Get StaticMeshEditorSubsystem(), StaticMesh:Static Mesh Object Reference=@n1) -> (ReturnValue:"Array of Float (single-precision)s")) -> (Array Element:float)
   n3 = Get l_count() -> (l_count=0)
   PrintString(
    execute=@n2.Completed
    InString=Append!(A=GetDisplayName(Object=@n1), B=" LOD Count ", C:String="To String (Integer)"(InInt=@n3)) -> (ReturnValue)
   )
   > Return Node!(execute, count:int=@n3)
   Branch(execute=@n2.LoopBody, Condition="float > float"(A=@n2))
   > Increment Int(Value=Get l_count() -> (l_count=0))
  ]}
]
variables [
 AssetsToOpenPostBuild {type="array<Object>", default []}
 CurrentLayoutPackageName {type=string, default=""}
 CurrentTargetMesh {type=DynamicMesh, default=null}
 isTesting {type=boolean, default=true}
 KillOnComplete {type=boolean}
 LayoutsToBuild {type="array<string>", default []}
 LODScreenSizes {type="array<number>", default [2,0.5,0.15,0.05]}
 RunOnSelectedLayouts {type=boolean}
 RunOnTestLayout {type=boolean}
 SavePath {type=string, default=/DelMarGame/Track/Meshes/}
 SavePathToSearch {type=string, default=&Track/}
 SavePrefix_Layout {type=string, default=Layout_}
 SavePrefix_Mesh {type=string, default=SM_DelMarTrack_}
]
