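; Editor function for the DelMar track combiner that derives the counterpart layout or mesh package name from a given asset name.
; It swaps save prefixes, optionally appends a collision suffix, and builds the folder path from the save root, asset type and testing flag.
; For layout input it searches static meshes recursively by name, otherwise or on no match it uses the built path.
; It returns whether that package exists, plus its name.
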(fn GetLayoutOrMeshNamesFromOther
  (params (assetName :type String) (InputIsMesh :type Boolean) (isTesting :type Boolean) (isCollision :type Boolean) (SavePathRoot :type String))
  (bind _returnvalue (AssetRegistry|GetAssetRegistry))
  (LocalVariables|Setl_assetName assetName)
  (LocalVariables|Setl_isMeshInput InputIsMesh)
  (LocalVariables|Setl_isTesting isTesting)
  (LocalVariables|Setl_isCollision isCollision)
  (LocalVariables|Setl_SavePathRoot SavePathRoot)
  (LocalVariables|Setl_assetName (Utilities|String|Append (select (LocalVariables|Getl_isMeshInput) (Utilities|String|Replace (LocalVariables|Getl_assetName) (LocalVariables|GetSavePrefix_Mesh) (LocalVariables|GetSavePrefix_Layout)) (Utilities|String|Replace (LocalVariables|Getl_assetName) (LocalVariables|GetSavePrefix_Layout) (LocalVariables|GetSavePrefix_Mesh))) (select (LocalVariables|Getl_isCollision) "_Collision" "")))
  (LocalVariables|Setl_assetPath (Utilities|String|Append (Utilities|String|Append (Utilities|String|Append (select (Utilities|String|IsEmpty (LocalVariables|Getl_SavePathRoot)) "/DelMarGame/Track/" (LocalVariables|Getl_SavePathRoot)) (select (and (Utilities|String|Equal,CaseInsensitive(String) (select (Utilities|String|IsEmpty (LocalVariables|Getl_SavePathRoot)) "/DelMarGame/Track/" (LocalVariables|Getl_SavePathRoot)) "/DelMarGame/Track/") (not (LocalVariables|Getl_isMeshInput))) "Desert/" "")) (select (LocalVariables|Getl_isMeshInput) "Layouts/" "Meshes/")) (select (LocalVariables|Getl_isTesting) "TempDontSubmit/" "")))
  (Utilities|FlowControl|Sequence
    (:then_0
      (if (not (LocalVariables|Getl_isMeshInput))
        ;[1 1 1] If layout was input search for a properly named mesh
        (bind _outassetdata (AssetRegistry|GetAssets _returnvalue (Utilities|MakeARFilter :PackagePaths (Utilities|Array|MakeArray (Utilities|String|StringToName (LocalVariables|Getl_assetPath))) :ClassPaths (Utilities|Array|MakeArray (Utilities|MakeTopLevelAssetPath "/Script/Engine" "StaticMesh")) :bRecursivePaths true :bRecursiveClasses false :bIncludeOnlyOnDiskAssets false) true))
        (LocalVariables|Setl_data _outassetdata)
        (if (not (Variables|GetisTesting))
          (AssetRegistry|UseFiltertoExcludeAssets _returnvalue (LocalVariables|Getl_data) (Utilities|MakeARFilter :PackagePaths (Utilities|Array|MakeArray "/DelMarGame/Track/Layouts" "/DelMarStainlessAssets/Environment/Tracks/Layouts") :bRecursivePaths true :bRecursiveClasses false :bIncludeOnlyOnDiskAssets false)))
        (bind _array_element (Utilities|Array|ForEachLoopwithBreak (LocalVariables|Getl_data))
          (:LoopBody
            (if (Utilities|String|Equal,CaseInsensitive(String) (Utilities|String|ToString(Name) (Utilities|Struct|BreakAssetData _array_element)) (LocalVariables|Getl_assetName))
              (LocalVariables|Setl_finalPackageName (Utilities|String|ToString(Name) (Utilities|Struct|BreakAssetData _array_element)))))
          (:Completed
            _)))
      ;[1 1 1] If mesh was input or If we didn't find a mesh then build a fallback path
      (if (Utilities|String|EqualExactly(String) (LocalVariables|Getl_finalPackageName))
        (LocalVariables|Setl_finalPackageName (Utilities|String|Append (LocalVariables|Getl_assetPath) (LocalVariables|Getl_assetName)))))
    (:then_1
      (bind _outassetdata_2 (AssetRegistry|GetAssetsbyPackageName (AssetRegistry|GetAssetRegistry) (Utilities|String|StringToName (LocalVariables|Getl_finalPackageName)) false true))
      (return (Utilities|Array|IsValidIndex _outassetdata_2) (LocalVariables|Getl_finalPackageName)))))
